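# Deprecated MetaSound patch that controls weapon fire audio timing, outputting gunshot triggers, current ammo and an auto-fire stop signal.
# In auto fire, a TriggerRepeat uses Weapon.FireRate as its period, starting after a short OnPlay delay and stopping after delayed On Stop, with ReTrigger accumulation able to restart it.
# A counter, reset on stop or ammo change, predicts remaining ammo to gate repeated shots.
# Non-auto fire triggers gunshots directly, reporting supplied ammo.
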MetaSoundPatch /FortMetaSoundWeapon/Items/Metasounds/Subgraphs/Deprecated/MSP_Weapon_Fire_Control
inputs [
 bAutoFire {type=Bool}
 On Stop {type=Trigger}
 OnPlay {type=Trigger}
 ReTrigger {type=Trigger}
 StopTriggerDelay {type=Time}
 Weapon.AmmoRemaining {type=Int32}
 Weapon.FireRate {type=Float}
]
outputs [
 CurrentAmmo {type=Int32}
 On Gunshot {type=Trigger}
 StopAutoFiring {type=Trigger}
]
pages [Default {
  input_defaults {On Stop=null, OnPlay=null, ReTrigger=null, StopTriggerDelay=0.03, Weapon.AmmoRemaining=30, Weapon.FireRate=1, bAutoFire=null}
  nodes [
   InitVariable.Trigger()
   n1 = TriggerCompare.Clamp:Bool(Compare=^ReTrigger, A=^bAutoFire) {also {defaults {B="<MetaSoundBuilderResult.SUCCEEDED: 0>"}}}
   TriggerDelay_1 = UE.Trigger Delay(In=^On Stop, Delay Time=^StopTriggerDelay)
   TriggerAccumulatorTriggerAccumulate2_0 = "TriggerAccumulator.Trigger Accumulate (2)"(In 0=@n1.True, In 1=@TriggerDelay_1.Out) {also {defaults {Auto Reset="<MetaSoundBuilderResult.SUCCEEDED: 0>"}}}
   TriggerDelay_0 = UE.Trigger Delay(In=^OnPlay) {also {defaults {Delay Time=0.01}}}
   n2 = TriggerCompare.Clamp:Bool(A=^bAutoFire, Compare=@TriggerDelay_0.Out) {also {defaults {B="<MetaSoundBuilderResult.SUCCEEDED: 0>"}}}
   TriggerRouteTriggerRouteBool3_0 = "TriggerRoute.Trigger Route (Bool, 3):Bool"(Set 2=@TriggerAccumulatorTriggerAccumulate2_0.Out, Set 0=@n2.True, Set 1=@TriggerDelay_1.Out) {also {defaults {Value 0="<MetaSoundBuilderResult.SUCCEEDED: 0>", Value 2="<MetaSoundBuilderResult.SUCCEEDED: 0>"}}}
   TriggerCompareClamp_2 = TriggerCompare.Clamp:Bool(Compare=@TriggerRouteTriggerRouteBool3_0.On Set, A=@TriggerRouteTriggerRouteBool3_0.Value) {also {defaults {B="<MetaSoundBuilderResult.SUCCEEDED: 0>"}}}
   TriggerOnValueChange_0 = UE.Trigger On Value Change:Int32(Value=^Weapon.AmmoRemaining)
   TriggerAnyTriggerAny2_2 = "TriggerAny.Trigger Any (2)"(In 1=UE.Reroute(Value=@TriggerCompareClamp_2.False).Value, In 0=@TriggerOnValueChange_0.Trigger)
   TriggerCounter_0 = UE.Trigger Counter(
    Reset=@TriggerAnyTriggerAny2_2.Out
    In=UE.TriggerRepeat(Period=UE.ConversionFloatToTime(In=^Weapon.FireRate).Out, Stop=@TriggerCompareClamp_2.False, Start=@TriggerCompareClamp_2.True).RepeatOut
   )
   TriggerAnyTriggerAny2_1 = "TriggerAny.Trigger Any (2)"(In 1=@n1.False, In 0=@n2.False)
   n3 = VariableAccessor.Trigger(Variable=VariableMutator.Trigger(Variable=InitVariable.Trigger().Variable, Value=@TriggerAnyTriggerAny2_1.Out).Variable)
   TriggerRouteTriggerRouteInt322_0 = "TriggerRoute.Trigger Route (Int32, 2):Int32"(
    Value 1=^Weapon.AmmoRemaining
    Value 0=UE.Subtract:Int32(
     PrimaryOperand=UE.Add:Int32(PrimaryOperand=^Weapon.AmmoRemaining) {also {defaults {AdditionalOperands=1}}}.Out
     AdditionalOperands=@TriggerCounter_0.Count
    ).Out
    Set 0=@TriggerCounter_0.On Trigger
    Set 1=@n3.Value
   )
   n4 = VariableMutator.Trigger(Variable=InitVariable.Trigger().Variable, Value=@TriggerCounter_0.On Trigger)
   TriggerAnyTriggerAny2_0 = "TriggerAny.Trigger Any (2)"(
    In 0=TriggerCompare.Clamp:Int32(A=@TriggerRouteTriggerRouteInt322_0.Value, Compare=VariableAccessor.Trigger(Variable=@n4.Variable).Value) {also {defaults {B=0, Type=5}}}.True
    In 1=VariableAccessor.Trigger(Variable=@n3.Variable).Value
   )
  ]
  links [
   {dst="@On Gunshot", src=TriggerAnyTriggerAny2_0.Out}
   {dst="@StopAutoFiring", src=TriggerCompareClamp_2.False}
   {dst="@CurrentAmmo", src=TriggerRouteTriggerRouteInt322_0.Value}
  ]
 }]
